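# Post-process style Anim Blueprint for the F_MED_Punk_Devil_Summer cosmetic body, following the attached parent mesh's pose.
# The output chain runs Copy Pose From Mesh through the body's Control Rig, then spring-driven Anim Dynamics on eight lace bones.
# A separate pelvis and calf corrective bone chain, weighted by thigh and knee yaw, is not connected to the output.
# The update event is also unconnected to the logic computing those weights.
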AnimBlueprint /BRCosmetics/Characters/Player/Female/Medium/Bodies/F_MED_Punk_Devil_Summer/Meshes/F_MED_Punk_Devil_Summer_AnimBP
target_skeleton=&/F_MED_Punk_Devil_Summer_Skeleton
graphs [
 AnimGraph {nodes [
   n1 = Get thigh_rot_r()
   n2 = Get thigh_rot_l()
   n3 = "Transform (Modify) Bone - Bone: pelvis_l"(Alpha=@n2, Translation="-7,0,0", Rotation="0,0,0", Scale="1,1.5,1") {
    scaleMode=BMM_Additive
    translationMode=BMM_Additive
    translationSpace=BCS_ParentBoneSpace
    alphaBoolBlend {blendOption=Linear}
    boneToModify {boneName=pelvis_l}
    scale {y=1.5}
    translation {x=-7}
   }
   n4 = "Transform (Modify) Bone - Bone: pelvis_mid_l"(ComponentPose=@n3, Alpha=@n2, Translation="-5,0,0", Rotation="0,0,0", Scale="1,1,1") {
    scaleMode=BMM_Additive
    translationMode=BMM_Additive
    translationSpace=BCS_ParentBoneSpace
    alphaBoolBlend {blendOption=Linear}
    boneToModify {boneName=pelvis_mid_l}
    translation {x=-5}
   }
   n5 = "Transform (Modify) Bone - Bone: pelvis_side_l"(ComponentPose=@n4, Alpha=@n2, Translation="0,0,-8", Rotation="0,0,0", Scale="1,1,1") {
    scaleMode=BMM_Additive
    translationMode=BMM_Additive
    translationSpace=BCS_ParentBoneSpace
    alphaBoolBlend {blendOption=Linear}
    boneToModify {boneName=pelvis_side_l}
    translation {z=-8}
   }
   n6 = "Transform (Modify) Bone - Bone: pelvis_r"(ComponentPose=@n5, Alpha=@n1, Translation="-12,0,0", Rotation="0,0,0", Scale="1,1.5,1") {
    scaleMode=BMM_Additive
    translationMode=BMM_Additive
    translationSpace=BCS_ParentBoneSpace
    alphaBoolBlend {blendOption=Linear}
    boneToModify {boneName=pelvis_r}
    scale {y=1.5}
    translation {x=-12}
   }
   n7 = "Transform (Modify) Bone - Bone: pelvis_mid_r"(ComponentPose=@n6, Alpha=@n1, Translation="-5,0,0", Rotation="0,0,0", Scale="1,1,1") {
    scaleMode=BMM_Additive
    translationMode=BMM_Additive
    translationSpace=BCS_ParentBoneSpace
    alphaBoolBlend {blendOption=Linear}
    boneToModify {boneName=pelvis_mid_r}
    translation {x=-5}
   }
   n8 = "Transform (Modify) Bone - Bone: pelvis_side_r"(ComponentPose=@n7, Alpha=@n1, Translation="0,0,8", Rotation="0,0,0", Scale="1,1,1") {
    scaleMode=BMM_Additive
    translationMode=BMM_Additive
    translationSpace=BCS_ParentBoneSpace
    alphaBoolBlend {blendOption=Linear}
    boneToModify {boneName=pelvis_side_r}
    translation {z=8}
   }
   n9 = "Transform (Modify) Bone - Bone: calf_l_split"(ComponentPose=@n8, Alpha=Get calf_rot_l() -> (calf_rot_l:double), Translation="-8,-5,0", Rotation="0,0,0", Scale="1,1,1") {
    translationMode=BMM_Additive
    translationSpace=BCS_ParentBoneSpace
    alphaBoolBlend {blendOption=Linear}
    boneToModify {boneName=calf_l_split}
    translation=<-8 -5>
   }
   "Transform (Modify) Bone - Bone: calf_r_split"(ComponentPose=@n9, Alpha=Get calf_rot_r() -> (calf_rot_r:double), Translation="8,5,0", Rotation="0,0,0", Scale="1,1,1") {
    translationMode=BMM_Additive
    translationSpace=BCS_ParentBoneSpace
    alphaBoolBlend {blendOption=Linear}
    boneToModify {boneName=calf_r_split}
    translation=<8 5>
   }
   n10 = Control Rig(Source=Copy Pose From Mesh() {bUseAttachedParent=true}) {
    alphaBoolBlend {blendOption=Linear}
    controlRigAssetReference {blueprintRigClass=&/F_MED_Punk_Devil_Summer_CtrlRig.F_MED_Punk_Devil_Summer_CtrlRig_C}
   }
   n11 = Anim Dynamics - Bone: laces_1_l(ComponentPose=Local To Component(LocalPose=@n10), Alpha=1, ExternalForce="0,0,0") {
    angularDampingOverride=1
    angularSpringConstant=1500
    bAngularSpring=true
    bGravityOverrideInSimSpace=true
    bOverrideAngularDamping=true
    gravityScale=0.35
    simulationSpace=RootRelative
    alphaBoolBlend {blendOption=Linear}
    alphaScaleBiasClamp {scale=0.4}
    boundBone {boneName=laces_1_l}
    physicsBodyDefinitions [{
      sphereCollisionRadius=0
      boundBone {boneName=laces_1_l}
      boxExtents=<10 20 20>
      constraintSetup {angularLimitsMax=<0 45 0>, angularLimitsMin=<0 -45 -90>, angularTarget=<0 0 0>}
      localJointOffset=<10 0 0>
     }]
   }
   n12 = Anim Dynamics - Bone: laces_2_l(ComponentPose=@n11, Alpha=1, ExternalForce="0,0,0") {
    angularDampingOverride=1
    angularSpringConstant=1500
    bAngularSpring=true
    bGravityOverrideInSimSpace=true
    bOverrideAngularDamping=true
    gravityScale=0.35
    simulationSpace=RootRelative
    alphaBoolBlend {blendOption=Linear}
    alphaScaleBiasClamp {scale=0.4}
    boundBone {boneName=laces_2_l}
    physicsBodyDefinitions [{
      sphereCollisionRadius=0
      boundBone {boneName=laces_2_l}
      boxExtents=<10 20 20>
      constraintSetup {angularLimitsMax=<0 45 0>, angularLimitsMin=<0 -45 -90>, angularTarget=<0 0 0>}
      localJointOffset=<10 0 0>
     }]
   }
   n13 = Anim Dynamics - Bone: laces_3_l(ComponentPose=@n12, Alpha=1, ExternalForce="0,0,0") {
    angularDampingOverride=1
    angularSpringConstant=1500
    bAngularSpring=true
    bGravityOverrideInSimSpace=true
    bOverrideAngularDamping=true
    gravityScale=0.35
    simulationSpace=RootRelative
    alphaBoolBlend {blendOption=Linear}
    alphaScaleBiasClamp {scale=0.4}
    boundBone {boneName=laces_3_l}
    physicsBodyDefinitions [{
      sphereCollisionRadius=0
      boundBone {boneName=laces_3_l}
      boxExtents=<10 20 20>
      constraintSetup {angularLimitsMax=<0 45 0>, angularLimitsMin=<0 -45 -90>, angularTarget=<0 0 0>}
      localJointOffset=<10 0 0>
     }]
   }
   n14 = Anim Dynamics - Bone: laces_4_l(ComponentPose=@n13, Alpha=1, ExternalForce="0,0,0") {
    angularDampingOverride=1
    angularSpringConstant=1500
    bAngularSpring=true
    bGravityOverrideInSimSpace=true
    bOverrideAngularDamping=true
    gravityScale=0.35
    simulationSpace=RootRelative
    alphaBoolBlend {blendOption=Linear}
    alphaScaleBiasClamp {scale=0.4}
    boundBone {boneName=laces_4_l}
    physicsBodyDefinitions [{
      sphereCollisionRadius=0
      boundBone {boneName=laces_4_l}
      boxExtents=<10 20 20>
      constraintSetup {angularLimitsMax=<0 45 0>, angularLimitsMin=<0 -45 -90>, angularTarget=<0 0 0>}
      localJointOffset=<10 0 0>
     }]
   }
   n15 = Anim Dynamics - Bone: laces_1_r(ComponentPose=@n14, Alpha=1, ExternalForce="0,0,0") {
    angularDampingOverride=1
    angularSpringConstant=1500
    bAngularSpring=true
    bGravityOverrideInSimSpace=true
    bOverrideAngularDamping=true
    gravityScale=0.35
    simulationSpace=RootRelative
    alphaBoolBlend {blendOption=Linear}
    alphaScaleBiasClamp {scale=0.4}
    boundBone {boneName=laces_1_r}
    physicsBodyDefinitions [{
      sphereCollisionRadius=0
      boundBone {boneName=laces_1_r}
      boxExtents=<10 20 20>
      constraintSetup {angularLimitsMax=<0 45 0>, angularLimitsMin=<0 -45 -90>, angularTarget=<0 0 0>}
      localJointOffset=<10 0 0>
     }]
   }
   n16 = Anim Dynamics - Bone: laces_2_r(ComponentPose=@n15, Alpha=1, ExternalForce="0,0,0") {
    angularDampingOverride=1
    angularSpringConstant=1500
    bAngularSpring=true
    bGravityOverrideInSimSpace=true
    bOverrideAngularDamping=true
    gravityScale=0.35
    simulationSpace=RootRelative
    alphaBoolBlend {blendOption=Linear}
    alphaScaleBiasClamp {scale=0.4}
    boundBone {boneName=laces_2_r}
    physicsBodyDefinitions [{
      sphereCollisionRadius=0
      boundBone {boneName=laces_2_r}
      boxExtents=<10 20 20>
      constraintSetup {angularLimitsMax=<0 45 0>, angularLimitsMin=<0 -45 -90>, angularTarget=<0 0 0>}
      localJointOffset=<10 0 0>
     }]
   }
   n17 = Anim Dynamics - Bone: laces_3_r(ComponentPose=@n16, Alpha=1, ExternalForce="0,0,0") {
    angularDampingOverride=1
    angularSpringConstant=1500
    bAngularSpring=true
    bGravityOverrideInSimSpace=true
    bOverrideAngularDamping=true
    gravityScale=0.35
    simulationSpace=RootRelative
    alphaBoolBlend {blendOption=Linear}
    alphaScaleBiasClamp {scale=0.4}
    boundBone {boneName=laces_3_r}
    physicsBodyDefinitions [{
      sphereCollisionRadius=0
      boundBone {boneName=laces_3_r}
      boxExtents=<10 20 20>
      constraintSetup {angularLimitsMax=<0 45 0>, angularLimitsMin=<0 -45 -90>, angularTarget=<0 0 0>}
      localJointOffset=<10 0 0>
     }]
   }
   n18 = Anim Dynamics - Bone: laces_4_r(ComponentPose=@n17, Alpha=1, ExternalForce="0,0,0") {
    angularDampingOverride=1
    angularSpringConstant=1500
    bAngularSpring=true
    bGravityOverrideInSimSpace=true
    bOverrideAngularDamping=true
    gravityScale=0.35
    simulationSpace=RootRelative
    alphaBoolBlend {blendOption=Linear}
    alphaScaleBiasClamp {scale=0.4}
    boundBone {boneName=laces_4_r}
    physicsBodyDefinitions [{
      sphereCollisionRadius=0
      boundBone {boneName=laces_4_r}
      boxExtents=<10 20 20>
      constraintSetup {angularLimitsMax=<0 45 0>, angularLimitsMin=<0 -45 -90>, angularTarget=<0 0 0>}
      localJointOffset=<10 0 0>
     }]
   }
   Output Pose(Result=Component To Local(ComponentPose=@n18)) {layerGroup=None}
  ]}
 EventGraph {nodes [
   n1 = Sequence()
   n2 = BreakTransform(InTransform=GetDeltaTransformFromRefPose(self=GetOwnerSkeletalMeshComponent(), BoneName=thigh_r, BaseName=pelvis))
   > Set thigh_rot_r(thigh_rot_r=MapRangeClamped(Value=BreakRotator(InRot=@n2.Rotation).Yaw, InRangeA=0, InRangeB=-90))
   n3 = BreakTransform(InTransform=GetDeltaTransformFromRefPose(self=GetOwnerSkeletalMeshComponent(), BoneName=thigh_l, BaseName=pelvis))
   > Set thigh_rot_l(thigh_rot_l=MapRangeClamped(Value=BreakRotator(InRot=@n3.Rotation).Yaw, InRangeA=-15, InRangeB=-110))
   n4 = BreakTransform(InTransform=GetDeltaTransformFromRefPose(self=GetOwnerSkeletalMeshComponent(), BoneName=calf_r, BaseName=thigh_r))
   Set calf_rot_r(execute=@n1.then_1, calf_rot_r:double=MapRangeClamped(Value=BreakRotator(InRot=@n4.Rotation).Yaw, InRangeA=45, InRangeB=110)) -> (Output_Get:double)
   n5 = BreakTransform(InTransform=GetDeltaTransformFromRefPose(self=GetOwnerSkeletalMeshComponent(), BoneName=calf_l, BaseName=thigh_l))
   > Set calf_rot_l(calf_rot_l:double=MapRangeClamped(Value=BreakRotator(InRot=@n5.Rotation).Yaw, InRangeA=45, InRangeB=110)) -> (Output_Get:double)
   Event BlueprintUpdateAnimation()
  ]}
]
variables [
 calf_rot_l {type=number}
 calf_rot_r {type=number}
 thigh_rot_l {type=number}
 thigh_rot_r {type=number}
]
