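# Post-process style cosmetic AnimBlueprint for the female medium Punk Devil Summer body part, adding secondary motion to hair and earrings.
# It copies the attached parent mesh's pose, then chains RootRelative AnimDynamics springs for buns, bangs and earrings, plus a small glasses bone offset.
# Per-bone alpha scales grade the sway, with bangs_all scaled to zero.
# Head-driven spherical limits constrain several bangs bones; the event graph has no logic.
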AnimBlueprint /BRCosmetics/Characters/Player/Female/Medium/Bodies/F_MED_Punk_Devil_Summer/Meshes/Parts/F_MED_Punk_Devil_Summer_AnimBP
target_skeleton=&/F_MED_Punk_Devil_Summer_Skeleton
graphs [
 AnimGraph {nodes [
   n1 = Anim Dynamics - Bone: bun_l(ComponentPose=Local To Component(LocalPose=Copy Pose From Mesh() {bUseAttachedParent=true}), Alpha=1, ExternalForce="0,0,0") {
    angularDampingOverride=1
    angularSpringConstant=1500
    bAngularSpring=true
    bGravityOverrideInSimSpace=true
    bOverrideAngularDamping=true
    gravityScale=0.35
    simulationSpace=RootRelative
    alphaBoolBlend {blendOption=Linear}
    alphaScaleBiasClamp {scale=0.08}
    boundBone {boneName=bun_l}
    physicsBodyDefinitions [{
      sphereCollisionRadius=0
      boundBone {boneName=bun_l}
      boxExtents=<10 20 20>
      constraintSetup {angularLimitsMax=<0 90 90>, angularLimitsMin=<0 -90 -90>, angularTarget=<0 0 0>}
      localJointOffset=<10 0 0>
     }]
   }
   n2 = Anim Dynamics - Bone: bun_r(ComponentPose=@n1, Alpha=1, ExternalForce="0,0,0") {
    angularDampingOverride=1
    angularSpringConstant=1500
    bAngularSpring=true
    bGravityOverrideInSimSpace=true
    bOverrideAngularDamping=true
    gravityScale=0.35
    simulationSpace=RootRelative
    alphaBoolBlend {blendOption=Linear}
    alphaScaleBiasClamp {scale=0.08}
    boundBone {boneName=bun_r}
    physicsBodyDefinitions [{
      sphereCollisionRadius=0
      boundBone {boneName=bun_r}
      boxExtents=<10 20 20>
      constraintSetup {angularLimitsMax=<0 90 90>, angularLimitsMin=<0 -90 -90>, angularTarget=<0 0 0>}
      localJointOffset=<10 0 0>
     }]
   }
   n3 = Anim Dynamics - Bone: bun_c(ComponentPose=@n2, Alpha=1, ExternalForce="0,0,0") {
    angularDampingOverride=1
    angularSpringConstant=1500
    bAngularSpring=true
    bGravityOverrideInSimSpace=true
    bOverrideAngularDamping=true
    gravityScale=0.35
    simulationSpace=RootRelative
    alphaBoolBlend {blendOption=Linear}
    alphaScaleBiasClamp {scale=0.08}
    boundBone {boneName=bun_c}
    physicsBodyDefinitions [{
      sphereCollisionRadius=0
      boundBone {boneName=bun_c}
      boxExtents=<10 20 20>
      constraintSetup {angularLimitsMax=<0 90 90>, angularLimitsMin=<0 -90 -90>, angularTarget=<0 0 0>}
      localJointOffset=<10 0 0>
     }]
   }
   n4 = "Transform (Modify) Bone - Bone: glasses"(ComponentPose=@n3, Alpha=1, Translation="0.3,0,0", Rotation="0,0,0", Scale="1,1,1") {
    translationMode=BMM_Additive
    translationSpace=BCS_BoneSpace
    alphaBoolBlend {blendOption=Linear}
    boneToModify {boneName=glasses}
    translation {x=0.3}
   }
   n5 = Anim Dynamics - Bone: bangs_1(ComponentPose=@n4, Alpha=1, ExternalForce="0,0,0") {
    angularDampingOverride=1
    angularSpringConstant=1500
    bAngularSpring=true
    bGravityOverrideInSimSpace=true
    bOverrideAngularDamping=true
    gravityScale=0.35
    simulationSpace=RootRelative
    alphaBoolBlend {blendOption=Linear}
    alphaScaleBiasClamp {scale=0.3}
    boundBone {boneName=bangs_1}
    componentLinearAccScale {y=0.45, z=0.45}
    componentLinearVelScale {y=0.45, z=0.45}
    physicsBodyDefinitions [{
      sphereCollisionRadius=0
      boundBone {boneName=bangs_1}
      boxExtents=<10 20 20>
      constraintSetup {angularLimitsMax=<0 90 0>, angularLimitsMin=<0 -90 -90>, angularTarget=<0 0 0>}
      localJointOffset=<10 0 0>
     }]
   }
   n6 = Anim Dynamics - Bone: bangs_2(ComponentPose=@n5, Alpha=1, ExternalForce="0,0,0") {
    angularDampingOverride=1
    angularSpringConstant=1500
    bAngularSpring=true
    bGravityOverrideInSimSpace=true
    bOverrideAngularDamping=true
    gravityScale=0.35
    simulationSpace=RootRelative
    alphaBoolBlend {blendOption=Linear}
    alphaScaleBiasClamp {scale=0.5}
    boundBone {boneName=bangs_2}
    componentLinearAccScale {y=0.45, z=0.45}
    componentLinearVelScale {y=0.45, z=0.45}
    physicsBodyDefinitions [{
      sphereCollisionRadius=0
      boundBone {boneName=bangs_2}
      boxExtents=<10 20 20>
      constraintSetup {angularLimitsMax=<0 90 0>, angularLimitsMin=<0 -90 -90>, angularTarget=<0 0 0>}
      localJointOffset=<10 0 0>
     }]
   }
   n7 = Anim Dynamics - Bone: bangs_3(ComponentPose=@n6, Alpha=1, ExternalForce="0,0,0") {
    angularDampingOverride=1
    angularSpringConstant=1500
    bAngularSpring=true
    bGravityOverrideInSimSpace=true
    bOverrideAngularDamping=true
    gravityScale=0.35
    simulationSpace=RootRelative
    alphaBoolBlend {blendOption=Linear}
    alphaScaleBiasClamp {scale=0.7}
    boundBone {boneName=bangs_3}
    componentLinearAccScale {y=0.45, z=0.45}
    componentLinearVelScale {y=0.45, z=0.45}
    physicsBodyDefinitions [{
      sphereCollisionRadius=0
      boundBone {boneName=bangs_3}
      boxExtents=<10 20 20>
      constraintSetup {angularLimitsMax=<0 90 90>, angularLimitsMin=<0 -90 -90>, angularTarget=<0 0 0>}
      localJointOffset=<10 0 0>
     }]
   }
   n8 = Anim Dynamics - Bone: bangs_4(ComponentPose=@n7, Alpha=1, ExternalForce="0,0,0") {
    angularDampingOverride=1
    angularSpringConstant=1500
    bAngularSpring=true
    bGravityOverrideInSimSpace=true
    bOverrideAngularDamping=true
    gravityScale=0.35
    simulationSpace=RootRelative
    alphaBoolBlend {blendOption=Linear}
    alphaScaleBiasClamp {scale=0.9}
    boundBone {boneName=bangs_4}
    componentLinearAccScale {y=0.45, z=0.45}
    componentLinearVelScale {y=0.45, z=0.45}
    physicsBodyDefinitions [{
      sphereCollisionRadius=0
      boundBone {boneName=bangs_4}
      boxExtents=<10 20 20>
      constraintSetup {angularLimitsMax=<0 90 90>, angularLimitsMin=<0 -90 -90>, angularTarget=<0 0 0>}
      localJointOffset=<10 0 0>
     }]
   }
   n9 = Anim Dynamics - Bone: bangs_all(ComponentPose=@n8, Alpha=1, ExternalForce="0,0,0") {
    angularDampingOverride=1
    angularSpringConstant=1500
    bAngularSpring=true
    bGravityOverrideInSimSpace=true
    bOverrideAngularDamping=true
    bUseSphericalLimits=true
    gravityScale=0.35
    simulationSpace=RootRelative
    alphaBoolBlend {blendOption=Linear}
    alphaScaleBiasClamp {scale=0}
    boundBone {boneName=bangs_all}
    componentLinearAccScale {y=0.45, z=0.45}
    componentLinearVelScale {y=0.45, z=0.45}
    physicsBodyDefinitions [{
      collisionType=CustomSphere
      sphereCollisionRadius=2
      boundBone {boneName=bangs_all}
      boxExtents=<10 20 20>
      constraintSetup {angularLimitsMax=<0 120 120>, angularLimitsMin=<0 -120 -120>, angularTarget=<0 0 0>}
      localJointOffset=<4 0 0>
     }]
    sphericalLimits [{limitRadius=8, drivingBone {boneName=head}, sphereLocalOffset=<-5 0 0>}]
   }
   n10 = Anim Dynamics - Bone: bangs_all_1(ComponentPose=@n9, Alpha=1, ExternalForce="0,0,0") {
    angularDampingOverride=1
    angularSpringConstant=1500
    bAngularSpring=true
    bGravityOverrideInSimSpace=true
    bOverrideAngularDamping=true
    bUseSphericalLimits=true
    gravityScale=0.35
    simulationSpace=RootRelative
    alphaBoolBlend {blendOption=Linear}
    alphaScaleBiasClamp {scale=0.5}
    boundBone {boneName=bangs_all_1}
    componentLinearAccScale {y=0.45, z=0.45}
    componentLinearVelScale {y=0.45, z=0.45}
    physicsBodyDefinitions [{
      collisionType=CustomSphere
      sphereCollisionRadius=2
      boundBone {boneName=bangs_all_1}
      boxExtents=<10 20 20>
      constraintSetup {angularLimitsMax=<0 90 90>, angularLimitsMin=<0 -90 -90>, angularTarget=<0 0 0>}
      localJointOffset=<4 0 0>
     }]
    sphericalLimits [{limitRadius=8, drivingBone {boneName=head}, sphereLocalOffset=<1 0 0>}]
   }
   n11 = Anim Dynamics - Bone: bangs_fr_2(ComponentPose=@n10, Alpha=1, ExternalForce="0,0,0") {
    angularDampingOverride=1
    angularSpringConstant=1500
    bAngularSpring=true
    bGravityOverrideInSimSpace=true
    bOverrideAngularDamping=true
    bUseSphericalLimits=true
    gravityScale=0.35
    simulationSpace=RootRelative
    alphaBoolBlend {blendOption=Linear}
    alphaScaleBiasClamp {scale=0.5}
    boundBone {boneName=bangs_fr_2}
    componentLinearAccScale {y=0.45, z=0.45}
    componentLinearVelScale {y=0.45, z=0.45}
    physicsBodyDefinitions [{
      collisionType=CustomSphere
      sphereCollisionRadius=2
      boundBone {boneName=bangs_fr_2}
      boxExtents=<10 20 20>
      constraintSetup {angularLimitsMax=<0 90 0>, angularLimitsMin=<0 -90 -90>, angularTarget=<0 0 0>}
      localJointOffset=<4 0 0>
     }]
    sphericalLimits [{limitRadius=8, drivingBone {boneName=head}, sphereLocalOffset=<-5 2 0>}]
   }
   n12 = Anim Dynamics - Bone: bangs_fr_3(ComponentPose=@n11, Alpha=1, ExternalForce="0,0,0") {
    angularDampingOverride=1
    angularSpringConstant=1500
    bAngularSpring=true
    bGravityOverrideInSimSpace=true
    bOverrideAngularDamping=true
    bUseSphericalLimits=true
    gravityScale=0.35
    simulationSpace=RootRelative
    alphaBoolBlend {blendOption=Linear}
    alphaScaleBiasClamp {scale=0.5}
    boundBone {boneName=bangs_fr_3}
    componentLinearAccScale {y=0.45, z=0.45}
    componentLinearVelScale {y=0.45, z=0.45}
    physicsBodyDefinitions [{
      collisionType=CustomSphere
      sphereCollisionRadius=2
      boundBone {boneName=bangs_fr_3}
      boxExtents=<10 20 20>
      constraintSetup {angularLimitsMax=<0 90 0>, angularLimitsMin=<0 -90 -90>, angularTarget=<0 0 0>}
      localJointOffset=<4 0 0>
     }]
    sphericalLimits [{limitRadius=8, drivingBone {boneName=head}, sphereLocalOffset=<-5 2 0>}]
   }
   n13 = Anim Dynamics - Bone: bangs_fr_4(ComponentPose=@n12, Alpha=1, ExternalForce="0,0,0") {
    angularDampingOverride=1
    angularSpringConstant=1500
    bAngularSpring=true
    bGravityOverrideInSimSpace=true
    bOverrideAngularDamping=true
    bUseSphericalLimits=true
    gravityScale=0.35
    simulationSpace=RootRelative
    alphaBoolBlend {blendOption=Linear}
    alphaScaleBiasClamp {scale=0.5}
    boundBone {boneName=bangs_fr_4}
    componentLinearAccScale {y=0.45, z=0.45}
    componentLinearVelScale {y=0.45, z=0.45}
    physicsBodyDefinitions [{
      collisionType=CustomSphere
      sphereCollisionRadius=2
      boundBone {boneName=bangs_fr_4}
      boxExtents=<10 20 20>
      constraintSetup {angularLimitsMax=<0 90 0>, angularLimitsMin=<0 -90 -90>, angularTarget=<0 0 0>}
      localJointOffset=<4 0 0>
     }]
    sphericalLimits [{limitRadius=8, drivingBone {boneName=head}, sphereLocalOffset=<-5 2 0>}]
   }
   n14 = Anim Dynamics - Bone: bangs_fr_5(ComponentPose=@n13, Alpha=1, ExternalForce="0,0,0") {
    angularDampingOverride=1
    angularSpringConstant=1500
    bAngularSpring=true
    bGravityOverrideInSimSpace=true
    bOverrideAngularDamping=true
    bUseSphericalLimits=true
    gravityScale=0.35
    simulationSpace=RootRelative
    alphaBoolBlend {blendOption=Linear}
    alphaScaleBiasClamp {scale=0.7}
    boundBone {boneName=bangs_fr_5}
    componentLinearAccScale {y=0.45, z=0.45}
    componentLinearVelScale {y=0.45, z=0.45}
    physicsBodyDefinitions [{
      collisionType=CustomSphere
      sphereCollisionRadius=2
      boundBone {boneName=bangs_fr_5}
      boxExtents=<10 20 20>
      constraintSetup {angularLimitsMax=<0 90 0>, angularLimitsMin=<0 -90 -90>, angularTarget=<0 0 0>}
      localJointOffset=<4 0 0>
     }]
    sphericalLimits [{limitRadius=8, drivingBone {boneName=head}, sphereLocalOffset=<-5 2 0>}]
   }
   n15 = Anim Dynamics - Bone: bangs_fr_6(ComponentPose=@n14, Alpha=1, ExternalForce="0,0,0") {
    angularDampingOverride=1
    angularSpringConstant=1500
    bAngularSpring=true
    bGravityOverrideInSimSpace=true
    bOverrideAngularDamping=true
    bUseSphericalLimits=true
    gravityScale=0.35
    simulationSpace=RootRelative
    alphaBoolBlend {blendOption=Linear}
    alphaScaleBiasClamp {scale=0.7}
    boundBone {boneName=bangs_fr_6}
    componentLinearAccScale {y=0.45, z=0.45}
    componentLinearVelScale {y=0.45, z=0.45}
    physicsBodyDefinitions [{
      collisionType=CustomSphere
      sphereCollisionRadius=2
      boundBone {boneName=bangs_fr_6}
      boxExtents=<10 20 20>
      constraintSetup {angularLimitsMax=<0 90 0>, angularLimitsMin=<0 -90 -90>, angularTarget=<0 0 0>}
      localJointOffset=<4 0 0>
     }]
    sphericalLimits [{limitRadius=8, drivingBone {boneName=head}, sphereLocalOffset=<-5 2 0>}]
   }
   n16 = Anim Dynamics - Bone: bangs_bk_2(ComponentPose=@n15, Alpha=1, ExternalForce="0,0,0") {
    angularDampingOverride=1
    angularSpringConstant=1500
    bAngularSpring=true
    bGravityOverrideInSimSpace=true
    bOverrideAngularDamping=true
    bUseSphericalLimits=true
    gravityScale=0.35
    simulationSpace=RootRelative
    alphaBoolBlend {blendOption=Linear}
    alphaScaleBiasClamp {scale=0.5}
    boundBone {boneName=bangs_bk_2}
    componentLinearAccScale {y=0.45, z=0.45}
    componentLinearVelScale {y=0.45, z=0.45}
    physicsBodyDefinitions [{
      collisionType=CustomSphere
      sphereCollisionRadius=2
      boundBone {boneName=bangs_bk_2}
      boxExtents=<10 20 20>
      constraintSetup {angularLimitsMax=<0 90 10>, angularLimitsMin=<0 -90 -90>, angularTarget=<0 0 0>}
      localJointOffset=<4 0 0>
     }]
    sphericalLimits [{limitRadius=8, drivingBone {boneName=head}, sphereLocalOffset=<-5 2 0>}]
   }
   n17 = Anim Dynamics - Bone: bangs_bk_3(ComponentPose=@n16, Alpha=1, ExternalForce="0,0,0") {
    angularDampingOverride=1
    angularSpringConstant=1500
    bAngularSpring=true
    bGravityOverrideInSimSpace=true
    bOverrideAngularDamping=true
    bUseSphericalLimits=true
    gravityScale=0.35
    simulationSpace=RootRelative
    alphaBoolBlend {blendOption=Linear}
    alphaScaleBiasClamp {scale=0.7}
    boundBone {boneName=bangs_bk_3}
    componentLinearAccScale {y=0.45, z=0.45}
    componentLinearVelScale {y=0.45, z=0.45}
    physicsBodyDefinitions [{
      collisionType=CustomSphere
      sphereCollisionRadius=2
      boundBone {boneName=bangs_bk_3}
      boxExtents=<10 20 20>
      constraintSetup {angularLimitsMax=<0 90 10>, angularLimitsMin=<0 -90 -90>, angularTarget=<0 0 0>}
      localJointOffset=<4 0 0>
     }]
    sphericalLimits [{limitRadius=8, drivingBone {boneName=head}, sphereLocalOffset=<-5 2 0>}]
   }
   n18 = Anim Dynamics - Bone: bangs_bk_4(ComponentPose=@n17, Alpha=1, ExternalForce="0,0,0") {
    angularDampingOverride=1
    angularSpringConstant=1500
    bAngularSpring=true
    bGravityOverrideInSimSpace=true
    bOverrideAngularDamping=true
    bUseSphericalLimits=true
    gravityScale=0.35
    simulationSpace=RootRelative
    alphaBoolBlend {blendOption=Linear}
    alphaScaleBiasClamp {scale=0.7}
    boundBone {boneName=bangs_bk_4}
    componentLinearAccScale {y=0.45, z=0.45}
    componentLinearVelScale {y=0.45, z=0.45}
    physicsBodyDefinitions [{
      collisionType=CustomSphere
      sphereCollisionRadius=2
      boundBone {boneName=bangs_bk_4}
      boxExtents=<10 20 20>
      constraintSetup {angularLimitsMax=<0 90 10>, angularLimitsMin=<0 -90 -90>, angularTarget=<0 0 0>}
      localJointOffset=<4 0 0>
     }]
    sphericalLimits [{limitRadius=8, drivingBone {boneName=head}, sphereLocalOffset=<-5 2 0>}]
   }
   n19 = Anim Dynamics - Bone: bangs_bk_5(ComponentPose=@n18, Alpha=1, ExternalForce="0,0,0") {
    angularDampingOverride=1
    angularSpringConstant=1500
    bAngularSpring=true
    bGravityOverrideInSimSpace=true
    bOverrideAngularDamping=true
    bUseSphericalLimits=true
    gravityScale=0.35
    simulationSpace=RootRelative
    alphaBoolBlend {blendOption=Linear}
    alphaScaleBiasClamp {scale=0.9}
    boundBone {boneName=bangs_bk_5}
    componentLinearAccScale {y=0.45, z=0.45}
    componentLinearVelScale {y=0.45, z=0.45}
    physicsBodyDefinitions [{
      collisionType=CustomSphere
      sphereCollisionRadius=2
      boundBone {boneName=bangs_bk_5}
      boxExtents=<10 20 20>
      constraintSetup {angularLimitsMax=<0 90 10>, angularLimitsMin=<0 -90 -90>, angularTarget=<0 0 0>}
      localJointOffset=<4 0 0>
     }]
    sphericalLimits [{limitRadius=8, drivingBone {boneName=head}, sphereLocalOffset=<-5 2 0>}]
   }
   n20 = Anim Dynamics - Bone: earring_1_r(ComponentPose=@n19, Alpha=1, ExternalForce="0,0,0") {
    angularDampingOverride=1
    angularSpringConstant=1500
    bAngularSpring=true
    bGravityOverrideInSimSpace=true
    bOverrideAngularDamping=true
    gravityScale=0.35
    simulationSpace=RootRelative
    alphaBoolBlend {blendOption=Linear}
    alphaScaleBiasClamp {scale=0.3}
    boundBone {boneName=earring_1_r}
    physicsBodyDefinitions [{
      sphereCollisionRadius=0
      boundBone {boneName=earring_1_r}
      boxExtents=<10 20 20>
      constraintSetup {angularLimitsMax=<0 0 0>, angularLimitsMin=<0 0 -40>, angularTarget=<0 0 0>}
      localJointOffset=<10 0 0>
     }]
   }
   n21 = Anim Dynamics - Bone: earring_2_r(ComponentPose=@n20, Alpha=1, ExternalForce="0,0,0") {
    angularDampingOverride=1
    angularSpringConstant=1500
    bAngularSpring=true
    bGravityOverrideInSimSpace=true
    bOverrideAngularDamping=true
    gravityScale=0.35
    simulationSpace=RootRelative
    alphaBoolBlend {blendOption=Linear}
    alphaScaleBiasClamp {scale=0.7}
    boundBone {boneName=earring_2_r}
    componentLinearAccScale {y=0.45, z=0.45}
    componentLinearVelScale {y=0.45, z=0.45}
    physicsBodyDefinitions [{
      sphereCollisionRadius=0
      boundBone {boneName=earring_2_r}
      boxExtents=<10 20 20>
      constraintSetup {angularLimitsMax=<0 90 90>, angularLimitsMin=<0 -90 -90>, angularTarget=<0 0 0>}
      localJointOffset=<10 0 0>
     }]
   }
   n22 = Anim Dynamics - Bone: earring_1_l(ComponentPose=@n21, Alpha=1, ExternalForce="0,0,0") {
    angularDampingOverride=1
    angularSpringConstant=1500
    bAngularSpring=true
    bGravityOverrideInSimSpace=true
    bOverrideAngularDamping=true
    gravityScale=0.35
    simulationSpace=RootRelative
    alphaBoolBlend {blendOption=Linear}
    alphaScaleBiasClamp {scale=0.3}
    boundBone {boneName=earring_1_l}
    physicsBodyDefinitions [{
      sphereCollisionRadius=0
      boundBone {boneName=earring_1_l}
      boxExtents=<10 20 20>
      constraintSetup {angularLimitsMax=<0 0 0>, angularLimitsMin=<0 0 -40>, angularTarget=<0 0 0>}
      localJointOffset=<10 0 0>
     }]
   }
   n23 = Anim Dynamics - Bone: earring_2_l(ComponentPose=@n22, Alpha=1, ExternalForce="0,0,0") {
    angularDampingOverride=1
    angularSpringConstant=1500
    bAngularSpring=true
    bGravityOverrideInSimSpace=true
    bOverrideAngularDamping=true
    gravityScale=0.35
    simulationSpace=RootRelative
    alphaBoolBlend {blendOption=Linear}
    alphaScaleBiasClamp {scale=0.7}
    boundBone {boneName=earring_2_l}
    componentLinearAccScale {y=0.45, z=0.45}
    componentLinearVelScale {y=0.45, z=0.45}
    physicsBodyDefinitions [{
      sphereCollisionRadius=0
      boundBone {boneName=earring_2_l}
      boxExtents=<10 20 20>
      constraintSetup {angularLimitsMax=<0 90 90>, angularLimitsMin=<0 -90 -90>, angularTarget=<0 0 0>}
      localJointOffset=<10 0 0>
     }]
   }
   Output Pose(Result=Component To Local(ComponentPose=@n23)) {layerGroup=None}
  ]}
 EventGraph {nodes [Event BlueprintUpdateAnimation()]}
]
variables []
